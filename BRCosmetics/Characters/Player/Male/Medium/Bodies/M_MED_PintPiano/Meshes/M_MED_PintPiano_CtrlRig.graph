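# Control Rig for the PintPiano male medium body that adds corrective deformation to the body, clothing and accessory bones.
# A shared DeformRig function runs first, then sections adjust belt items, pant ends, watches, necklace, pocket and zipper.
# Animation attributes, spherical pose readers and joint rotations weight the offsets, alongside weighted parent constraints.
# The necklace aims at a setup-spawned bone, and a Kalman-smoothed Skydiving flag drives harmonic flutter on necklace and pants.
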ControlRigBlueprint /BRCosmetics/Characters/Player/Male/Medium/Bodies/M_MED_PintPiano/Meshes/M_MED_PintPiano_CtrlRig
graphs [
 CollapseNode_1_ContainedGraph {nodes [
   n1 = RigVMFunctionEntryNode()
   Get_Animation_Attribute = RigVMDispatchNode(BoneName=None, Default=0, Name=deformRig_l_thigh_up_alpha) {template="DISPATCH_RigDispatch_GetAnimAttribute(in Name,in BoneName,in Default,out Value,out Found)"}
   &RigUnit_ModifyTransforms(
    ItemToModify=((Item=(Type=Bone,Name="dyn_belt_pouch_base_l"),Transform=(Rotation=(X=0.2341,Y=-0.1273,Z=-0.0026,W=0.9638),Translation=(X=-1.282,Y=-3.2351,Z=-0.261),Scale3D=(X=1,Y=1,Z=1))))
    WeightMinimum=0
    ExecutePin=@n1.ExecutePin
    Weight=@Get_Animation_Attribute.Value
   )
   Get_Animation_Attribute_1 = RigVMDispatchNode(BoneName=None, Default=0, Name=deformRig_r_thigh_up_alpha) {template="DISPATCH_RigDispatch_GetAnimAttribute(in Name,in BoneName,in Default,out Value,out Found)"}
   > &RigUnit_ModifyTransforms(
    ItemToModify="((Item=(Type=Bone,Name=\"dyn_tool_belt_base\"),Transform=(Rotation=(X=-0.1153,Y=-0.0075,Z=0.0093,W=0.9933),Translation=(X=-1.0318,Y=-2.542,Z=-1.868),Scale3D=(X=1,Y=1,Z=1))),(Item=(Type=Bone,Name=\"dyn_belt_base\"),Transform=(Rotation=(X=0.1505,Y=-0.0586,Z=-0.5 ...(truncated)"
    WeightMinimum=0
    Weight=@Get_Animation_Attribute_1.Value
   )
   > n2 = &RigUnit_SphericalPoseReader(
    ActiveRegionScaleFactors=(PositiveWidth=3,NegativeWidth=1,PositiveHeight=1.3,NegativeHeight=1.2)
    ActiveRegionSize=0.1
    Debug=(bDrawDebug=false,bDraw2D=False,bDrawLocalAxes=False,DebugScale=25,DebugSegments=20,DebugThickness=0.25)
    DriverAxis=(X=-1,Y=0,Z=0)
    DriverItem=(Type=Bone,Name="thigh_l")
    FalloffRegionScaleFactors=(PositiveWidth=1,NegativeWidth=1,PositiveHeight=0.5,NegativeHeight=0.5)
    FalloffSize=0.57
    FlipHeightScaling=False
    FlipWidthScaling=False
    OptionalParentItem=(Type=Bone,Name="None")
    RotationOffset=(X=130,Y=5,Z=-12)
   )
   > &RigUnit_ModifyTransforms(
    ItemToModify=((Item=(Type=Bone,Name="dyn_belt_pouch_base_l"),Transform=(Rotation=(X=0.2091,Y=-0.2527,Z=0.056,W=0.943),Translation=(X=0,Y=-2,Z=1),Scale3D=(X=1,Y=1,Z=1))))
    WeightMinimum=0
    Weight=@n2.OutputParam
   )
   > n3 = &RigUnit_SphericalPoseReader(
    ActiveRegionScaleFactors=(PositiveWidth=3,NegativeWidth=1,PositiveHeight=1,NegativeHeight=1)
    ActiveRegionSize=0.1
    Debug=(bDrawDebug=false,bDraw2D=False,bDrawLocalAxes=False,DebugScale=25,DebugSegments=20,DebugThickness=0.25)
    DriverAxis=(X=1,Y=0,Z=0)
    DriverItem=(Type=Bone,Name="thigh_r")
    FalloffRegionScaleFactors=(PositiveWidth=1,NegativeWidth=1,PositiveHeight=1,NegativeHeight=1)
    FalloffSize=0.5
    FlipHeightScaling=true
    FlipWidthScaling=true
    OptionalParentItem=(Type=Bone,Name="None")
    RotationOffset=(X=0,Y=0,Z=0)
   )
   > n4 = &RigUnit_ModifyTransforms(
    ItemToModify=((Item=(Type=Bone,Name="dyn_tool_belt_base"),Transform=(Rotation=(X=0.0311,Y=-0.63,Z=0.0389,W=-0.775),Translation=(X=-4.6496,Y=0.1674,Z=-3.6979),Scale3D=(X=1,Y=1,Z=1))))
    WeightMinimum=0
    Weight=@n3.OutputParam
   )
   RigVMFunctionReturnNode(ExecutePin=@n4.ExecutePin)
   Leg ROM = RigVMFunctionReferenceNode(
    ExecuteContext=()
    Include Ankles=true
    Include Distance Checks=true
    Include Knees=true
    Interp Output Speed=5
    Interp Results=true
    Result Alpha=1
    SPR Optional Parent Item=(Type=Bone,Name="None")
   )
  ]}
 CollapseNode_ContainedGraph {nodes [
   RigVMFunctionEntryNode()
   > RigVMFunction_Sequence = RigVMAggregateNode()
   Get Delta Ref = RigVMFunctionReferenceNode(Item=(Type=Bone,Name="foot_l"), ExecuteContext=@RigVMFunction_Sequence.A)
   > Get Delta Ref_3 = RigVMFunctionReferenceNode(Item=(Type=Bone,Name="foot_r"))
   Remap = &RigVMFunction_MathDoubleRemap(SourceMaximum=90, SourceMinimum=0, TargetMaximum=1, TargetMinimum=0, Value=@"Get Delta Ref".Result.Rotation.Yaw)
   > Modify Transforms with Mirror = RigVMFunctionReferenceNode(
    Base Modify Data=((Item=(Type=Bone,Name="deform_pants_lower_l"),Transform=(Rotation=(X=0,Y=0,Z=0.2588,W=0.9659),Translation=(X=2.5,Y=1.5,Z=0),Scale3D=(X=1,Y=1.05,Z=1))))
    Mirror Modify Data=((Item=(Type=Bone,Name="deform_pants_rower_r"),Transform=(Rotation=(X=0,Y=0,Z=0,W=1),Translation=(X=0,Y=0,Z=0),Scale3D=(X=1,Y=1,Z=1))))
    Weight=@Remap.Result
    "Mirror Weight "=&RigVMFunction_MathDoubleRemap(SourceMaximum=90, SourceMinimum=0, TargetMaximum=1, TargetMinimum=0, Value=@"Get Delta Ref_3".Result.Rotation.Yaw).Result
   )
   Get Delta Ref_1 = RigVMFunctionReferenceNode(Item=(Type=Bone,Name="foot_l"), ExecuteContext=@RigVMFunction_Sequence.B)
   > Get Delta Ref_2 = RigVMFunctionReferenceNode(Item=(Type=Bone,Name="foot_r"))
   > Modify Transforms with Mirror_1 = RigVMFunctionReferenceNode(
    Base Modify Data=((Item=(Type=Bone,Name="deform_pants_lower_l"),Transform=(Rotation=(X=0,Y=0,Z=-0.1736,W=0.9848),Translation=(X=3,Y=-4,Z=0),Scale3D=(X=1,Y=1.4,Z=1.2))))
    Mirror Modify Data=((Item=(Type=Bone,Name="deform_pants_rower_r"),Transform=(Rotation=(X=0,Y=0,Z=0,W=1),Translation=(X=0,Y=0,Z=0),Scale3D=(X=1,Y=1,Z=1))))
    Weight=&RigVMFunction_MathDoubleRemap(SourceMaximum=-90, SourceMinimum=0, TargetMaximum=1, TargetMinimum=0, Value=@"Get Delta Ref_1".Result.Rotation.Yaw).Result
    "Mirror Weight "=&RigVMFunction_MathDoubleRemap(SourceMaximum=-90, SourceMinimum=0, TargetMaximum=1, TargetMinimum=0, Value=@"Get Delta Ref_2".Result.Rotation.Yaw).Result
   )
   Modify Transforms with Mirror_2 = RigVMFunctionReferenceNode(
    Base Modify Data=((Item=(Type=Bone,Name="deform_pants_lower_l"),Transform=(Rotation=(X=0,Y=0,Z=-0.0262,W=0.9997),Translation=(X=0,Y=-0.5,Z=0),Scale3D=(X=1,Y=1.075,Z=1.075))))
    Mirror Modify Data=((Item=(Type=Bone,Name="deform_pants_rower_r"),Transform=(Rotation=(X=0,Y=0,Z=0,W=1),Translation=(X=0,Y=0,Z=0),Scale3D=(X=1,Y=1,Z=1))))
    "Mirror Weight "=1
    Weight=1
    ExecuteContext=@RigVMFunction_Sequence.C
   )
   Cast = &RigVMFunction_MathBoolToFloat(Value=RigVMVariableNode() {variable=Skydiving}.Value)
   KalmanFilter = &RigVMFunction_KalmanFloat(BufferSize=16, Value=@Cast.Result) {template="KalmanFilter::Execute(in Value,in BufferSize,out Result)"}
   Multiply = &RigVMFunction_MathDoubleMul(B=1, A=@KalmanFilter.Result)
   > /Script/ControlRig.RigUnit_ItemHarmonics(
    RotationOrder=YZX
    Targets=((Item=(Type=Bone,Name="deform_pants_lower_l"),Ratio=0.3),(Item=(Type=Bone,Name="deform_pants_rower_r"),Ratio=0.2))
    WaveAmplitude=(X=0,Y=10,Z=10)
    WaveEase=Linear
    WaveFrequency=(X=1,Y=1,Z=0.8)
    WaveNoise=(X=0,Y=3,Z=3)
    WaveOffset=(X=0,Y=12,Z=6)
    WaveSpeed=(X=4,Y=4,Z=4)
    WaveMaximum=@Multiply.Result
    WaveMinimum=&RigVMFunction_MathDoubleMul(B=0.5, A=@KalmanFilter.Result).Result
   ) {is=struct RigVMUnitNode}
   n1 = &RigVMFunction_KalmanFloat(BufferSize=16, Value=&RigVMFunction_MathBoolToFloat(Value=RigVMVariableNode() {variable=Skydiving}.Value).Result) {template="KalmanFilter::Execute(in Value,in BufferSize,out Result)"}
   > Modify Transforms with Mirror_3 = RigVMFunctionReferenceNode(
    Base Modify Data=((Item=(Type=Bone,Name="deform_pants_lower_l"),Transform=(Rotation=(X=0,Y=0,Z=0,W=1),Translation=(X=2,Y=0,Z=0),Scale3D=(X=1,Y=1.1,Z=1.1))))
    Mirror Modify Data=((Item=(Type=Bone,Name="deform_pants_rower_r"),Transform=(Rotation=(X=0,Y=0,Z=0,W=1),Translation=(X=0,Y=0,Z=0),Scale3D=(X=1,Y=1,Z=1))))
    "Mirror Weight "=@n1.Result
    Weight=@n1.Result
   )
   RigVMFunctionReturnNode(ExecuteContext=@RigVMFunction_Sequence.D)
  ]}
 CollapseNode_ContainedGraph {nodes [
   n1 = RigVMFunctionEntryNode()
   Get_Animation_Attribute = RigVMDispatchNode(BoneName=None, Default=0, Name=deformRig_l_wrist_up_alpha) {template="DISPATCH_RigDispatch_GetAnimAttribute(in Name,in BoneName,in Default,out Value,out Found)"}
   &RigUnit_ModifyTransforms(
    ItemToModify=((Item=(Type=Bone,Name="deform_watch_l"),Transform=(Rotation=(X=0,Y=0.0969,Z=0,W=-0.9953),Translation=(X=-0.8896,Y=0,Z=-0.2025),Scale3D=(X=1,Y=1,Z=1.0486))))
    WeightMinimum=0
    ExecutePin=@n1.ExecutePin
    Weight=@Get_Animation_Attribute.Value
   )
   Get_Animation_Attribute_2 = RigVMDispatchNode(BoneName=None, Default=0, Name=deformRig_r_wrist_up_alpha) {template="DISPATCH_RigDispatch_GetAnimAttribute(in Name,in BoneName,in Default,out Value,out Found)"}
   > &RigUnit_ModifyTransforms(
    ItemToModify=((Item=(Type=Bone,Name="deform_watch_r"),Transform=(Rotation=(X=0,Y=0.0399,Z=0,W=-0.9992),Translation=(X=0.9034,Y=0,Z=-0.04),Scale3D=(X=1,Y=1,Z=1.087))))
    WeightMinimum=0
    Weight=@Get_Animation_Attribute_2.Value
   )
   Get_Animation_Attribute_1 = RigVMDispatchNode(BoneName=None, Default=0, Name=deformRig_l_wrist_dn_alpha) {template="DISPATCH_RigDispatch_GetAnimAttribute(in Name,in BoneName,in Default,out Value,out Found)"}
   > &RigUnit_ModifyTransforms(
    ItemToModify=((Item=(Type=Bone,Name="deform_watch_l"),Transform=(Rotation=(X=0,Y=0,Z=0,W=1),Translation=(X=0,Y=0,Z=0),Scale3D=(X=1,Y=1,Z=1))))
    WeightMinimum=0
    Weight=@Get_Animation_Attribute_1.Value
   )
   Get_Animation_Attribute_3 = RigVMDispatchNode(BoneName=None, Default=0, Name=deformRig_r_wrist_dn_alpha) {template="DISPATCH_RigDispatch_GetAnimAttribute(in Name,in BoneName,in Default,out Value,out Found)"}
   > n2 = &RigUnit_ModifyTransforms(
    ItemToModify=((Item=(Type=Bone,Name="deform_watch_r"),Transform=(Rotation=(X=0,Y=0,Z=0,W=1),Translation=(X=0,Y=0,Z=0),Scale3D=(X=1,Y=1,Z=1))))
    WeightMinimum=0
    Weight=@Get_Animation_Attribute_3.Value
   )
   RigVMFunctionReturnNode(ExecutePin=@n2.ExecutePin)
  ]}
 CollapseNode_ContainedGraph {nodes [
   n1 = RigVMFunctionEntryNode()
   &RigUnit_ParentConstraint(
    AdvancedSettings=(RotationOrderForFilter=XZY)
    Parents="((Item=(Type=Bone,Name=\"deform_pec_r\"),Weight=0.0005),(Item=(Type=Bone,Name=\"clavicle_l\"),Weight=0.0005),(Item=(Type=Bone,Name=\"clavicle_r\"),Weight=0.0005),(Item=(Type=Bone,Name=\"spine_05\"),Weight=0.0461),(Item=(Type=Bone,Name=\"spine_01\"),Weight=0.0466),(I ...(truncated)"
    Weight=1
    ExecutePin=@n1.ExecutePin
    Child=RigVMVariableNode() {variable=Spawned_Necklace_End}.Value
   )
   > &RigUnit_ParentConstraint(
    AdvancedSettings=(RotationOrderForFilter=XZY)
    Child=(Type=Bone,Name="cr_stomach_target")
    Parents=((Item=(Type=Bone,Name="spine_04"),Weight=0.015),(Item=(Type=Bone,Name="spine_03"),Weight=0.1275),(Item=(Type=Bone,Name="spine_01"),Weight=0.1652),(Item=(Type=Bone,Name="spine_02"),Weight=0.2043),(Item=(Type=Bone,Name="pelvis"),Weight=0.488))
    Weight=1
   )
   n2 = RigVMVariableNode() {variable=Spawned_Necklace_End}
   > &RigUnit_AimConstraintLocalSpaceOffset(
    Child=(Type=Bone,Name="dyn_necklace_base")
    Filter=(bX=True,bY=True,bZ=True)
    Parents=((Item=(Type=Bone,Name="None"),Weight=1))
    Weight=1
    WorldUp=(Target=(X=0,Y=0,Z=1),Kind=Location,Space=(Type=None,Name="None"))
    Parents.0.Item=@n2.Value
    WorldUp.Space=@n2.Value
   )
   > &RigUnit_ModifyTransforms(
    ItemToModify=((Item=(Type=Bone,Name="dyn_necklace_base"),Transform=(Rotation=(X=0,Y=0,Z=0,W=-1),Translation=(X=0,Y=0.7964,Z=0),Scale3D=(X=1,Y=1,Z=1))))
    Weight=1
    WeightMinimum=0
   )
   > n3 = &RigUnit_SphericalPoseReader(
    ActiveRegionScaleFactors=(PositiveWidth=1,NegativeWidth=1,PositiveHeight=1,NegativeHeight=1)
    ActiveRegionSize=0.1
    Debug=(bDrawDebug=true,bDraw2D=False,bDrawLocalAxes=False,DebugScale=25,DebugSegments=20,DebugThickness=0.25)
    DriverAxis=(X=0,Y=1,Z=0)
    DriverItem=(Type=Bone,Name="spine_05")
    FalloffRegionScaleFactors=(PositiveWidth=1,NegativeWidth=1,PositiveHeight=1,NegativeHeight=1)
    FalloffSize=0.2
    FlipHeightScaling=False
    FlipWidthScaling=False
    OptionalParentItem=(Type=Bone,Name="pelvis")
    RotationOffset=(X=30,Y=0,Z=0)
   )
   > &RigUnit_ModifyTransforms(
    ItemToModify=((Item=(Type=Bone,Name="dyn_necklace_base"),Transform=(Rotation=(X=0,Y=0,Z=0,W=-1),Translation=(X=0,Y=0.5,Z=1),Scale3D=(X=1,Y=1,Z=1))))
    WeightMinimum=0
    Weight=@n3.OutputParam
   )
   > n4 = &RigUnit_SphericalPoseReader(
    ActiveRegionScaleFactors=(PositiveWidth=1,NegativeWidth=1,PositiveHeight=1,NegativeHeight=1)
    ActiveRegionSize=0.1
    Debug=(bDrawDebug=true,bDraw2D=False,bDrawLocalAxes=False,DebugScale=25,DebugSegments=20,DebugThickness=0.25)
    DriverAxis=(X=0,Y=1,Z=0)
    DriverItem=(Type=Bone,Name="spine_04")
    FalloffRegionScaleFactors=(PositiveWidth=1,NegativeWidth=1,PositiveHeight=1,NegativeHeight=1)
    FalloffSize=0.22
    FlipHeightScaling=False
    FlipWidthScaling=False
    OptionalParentItem=(Type=Bone,Name="cr_stomach_target")
    RotationOffset=(X=90,Y=0,Z=50)
   )
   > &RigUnit_ModifyTransforms(
    ItemToModify="((Item=(Type=Bone,Name=\"dyn_necklace_base\"),Transform=(Rotation=(X=0,Y=0,Z=0.0436,W=0.999),Translation=(X=0,Y=0.5,Z=1),Scale3D=(X=1,Y=1,Z=1))),(Item=(Type=Bone,Name=\"dyn_necklace_02\"),Transform=(Rotation=(X=0,Y=0,Z=0.061,W=0.9981),Translation=(X=0,Y=0,Z=0) ...(truncated)"
    WeightMinimum=0
    Weight=&RigVMFunction_KalmanFloat(BufferSize=4, Value=@n4.OutputParam) {template="KalmanFilter::Execute(in Value,in BufferSize,out Result)"}.Result
   )
   Cast = &RigVMFunction_MathBoolToFloat(Value=RigVMVariableNode() {variable=Skydiving}.Value)
   KalmanFilter = &RigVMFunction_KalmanFloat(BufferSize=16, Value=@Cast.Result) {template="KalmanFilter::Execute(in Value,in BufferSize,out Result)"}
   Multiply = &RigVMFunction_MathDoubleMul(B=1, A=@KalmanFilter.Result)
   > &RigUnit_ItemHarmonics(
    RotationOrder=YZX
    Targets=((Item=(Type=Bone,Name="dyn_necklace_02"),Ratio=0.1),(Item=(Type=Bone,Name="dyn_necklace_03"),Ratio=1),(Item=(Type=Bone,Name="dyn_clipboard_02"),Ratio=0))
    WaveAmplitude=(X=0,Y=10,Z=10)
    WaveEase=Linear
    WaveFrequency=(X=1,Y=1,Z=0.8)
    WaveNoise=(X=0,Y=3,Z=3)
    WaveOffset=(X=0,Y=12,Z=6)
    WaveSpeed=(X=3,Y=3,Z=3)
    WaveMaximum=@Multiply.Result
    WaveMinimum=&RigVMFunction_MathDoubleMul(B=0.5, A=@KalmanFilter.Result).Result
   )
   > n5 = &RigUnit_ModifyTransforms(
    ItemToModify="((Item=(Type=Bone,Name=\"dyn_necklace_02\"),Transform=(Rotation=(X=0,Y=0,Z=0.0436,W=0.999),Translation=(X=0,Y=0,Z=0),Scale3D=(X=1,Y=1,Z=1))),(Item=(Type=Bone,Name=\"dyn_necklace_03\"),Transform=(Rotation=(X=0,Y=0,Z=0.1736,W=0.9848),Translation=(X=0,Y=0,Z=0),Sc ...(truncated)"
    WeightMinimum=0
    Weight=@KalmanFilter.Result
   )
   RigVMFunctionReturnNode(ExecutePin=@n5.ExecutePin)
  ]}
 CollapseNode_ContainedGraph {nodes [
   n1 = RigVMFunctionEntryNode()
   RigVMFunction_Sequence = &RigVMFunction_Sequence(ExecuteContext=@n1.ExecutePin)
   &RigUnit_ParentConstraint(
    AdvancedSettings=(RotationOrderForFilter=XZY)
    Child=(Type=Bone,Name="deform_pocket_l_base")
    Parents="((Item=(Type=Bone,Name=\"pelvis\"),Weight=0.0003),(Item=(Type=Bone,Name=\"spine_01\"),Weight=0.0019),(Item=(Type=Bone,Name=\"upperarm_twist_01_l\"),Weight=0.0173),(Item=(Type=Bone,Name=\"deform_upperarm_fr_l\"),Weight=0.0197),(Item=(Type=Bone,Name=\"clavicle_l\"),We ...(truncated)"
    Weight=1
    ExecutePin=@RigVMFunction_Sequence.A
   )
   > &RigUnit_ModifyTransforms(
    ItemToModify=((Item=(Type=Bone,Name="deform_pocket_l_base"),Transform=(Rotation=(X=0,Y=0,Z=0,W=1),Translation=(X=0,Y=0.55,Z=0),Scale3D=(X=1,Y=1,Z=1))))
    Weight=1
    WeightMinimum=0
   )
   Get_Animation_Attribute = RigVMDispatchNode(BoneName=None, Default=0, Name=deformRig_l_upperarm_fr_alpha) {template="DISPATCH_RigDispatch_GetAnimAttribute(in Name,in BoneName,in Default,out Value,out Found)"}
   Remap = &RigVMFunction_MathDoubleRemap(SourceMaximum=0.35, SourceMinimum=0, TargetMaximum=1, TargetMinimum=0, Value=@Get_Animation_Attribute.Value)
   KalmanFilter = &RigVMFunction_KalmanFloat(BufferSize=4, Value=@Remap.Result) {template="KalmanFilter::Execute(in Value,in BufferSize,out Result)"}
   > &RigUnit_ModifyTransforms(
    ItemToModify=((Item=(Type=Bone,Name="deform_pocket_l_base"),Transform=(Rotation=(X=-0.1736,Y=0,Z=0,W=0.9848),Translation=(X=0,Y=1,Z=-0.3),Scale3D=(X=1,Y=1,Z=1))))
    WeightMinimum=0
    Weight=@KalmanFilter.Result
   )
   Inverse = /Script/RigVM.RigVMFunction_MathTransformInverse(Value=&RigUnit_GetTransform(Item=(Type=Bone,Name="clavicle_l"), Space=LocalSpace, bInitial=true).Transform) {is=struct RigVMUnitNode, template="Inverse::Execute(in Value,out Result)"}
   RigUnit_GetTransform = &RigUnit_GetTransform(Item=(Type=Bone,Name="clavicle_l"), Space=LocalSpace, bInitial=False)
   Multiply = &RigVMFunction_MathTransformMul(B=@Inverse.Result, A=@RigUnit_GetTransform.Transform) {template="Multiply::Execute(in A,in B,out Result)"}
   RigVMFunction_MathQuaternionSwingTwist = &RigVMFunction_MathQuaternionSwingTwist(TwistAxis=(X=1,Y=0,Z=0), Input=@Multiply.Result.Rotation)
   RigVMFunction_MathQuaternionToEuler = &RigVMFunction_MathQuaternionToEuler(RotationOrder=ZYX, Value=@RigVMFunction_MathQuaternionSwingTwist.Twist)
   > &RigUnit_ModifyTransforms(
    ItemToModify=((Item=(Type=Bone,Name="deform_pocket_l_base"),Transform=(Rotation=(X=0,Y=0,Z=0.0436,W=0.999),Translation=(X=0,Y=0.5,Z=0),Scale3D=(X=1,Y=1,Z=1))))
    WeightMinimum=0
    Weight=&RigVMFunction_KalmanFloat(
     BufferSize=4
     Value=&RigVMFunction_MathDoubleRemap(SourceMaximum=0, SourceMinimum=-45, TargetMaximum=0, TargetMinimum=1, Value=@RigVMFunction_MathQuaternionToEuler.Result.X).Result
    ) {template="KalmanFilter::Execute(in Value,in BufferSize,out Result)"}.Result
   )
   RigVMFunctionReturnNode(ExecutePin=@RigVMFunction_Sequence.B)
  ]}
 RigVMFunctionLibrary {links [], nodes []}
 RigVMFunction_Sequence_1_ContainedGraph {nodes [
   n1 = RigVMFunctionEntryNode()
   RigVMFunction_Sequence = &RigVMFunction_Sequence(ExecuteContext=@n1.ExecuteContext)
   n2 = &RigVMFunction_Sequence(ExecuteContext=@RigVMFunction_Sequence.B)
   n3 = &RigVMFunction_Sequence(ExecuteContext=@n2.B)
   RigVMFunctionReturnNode(A=@RigVMFunction_Sequence.A, B=@n2.A, C=@n3.A, D=@n3.B)
  ]}
 RigVMModel {nodes [
   RigUnit_BeginExecution = &RigUnit_BeginExecution()
   FN_DeformRig_FNC = RigVMFunctionReferenceNode(
    DeformRig_ Curve="(deformRig_l_upperarm_up_alpha_72_#1=False,deformRig_r_upperarm_up_alpha_73_#2=False,deformRig_l_upperarm_fr_alpha_74_#3=True,deformRig_r_upperarm_fr_alpha_75_CB51B80 ...(truncated)"
    DeformRig_Bool="(upperarm_1_#4=True,elbow_3_#5=True,wrist_7_#6=True,groin_8_#7=True,glute_17_#8=True,knee_9_40C408324CA261 ...(truncated)"
    DeformRig_Debug=(upperarm_debug_18_#4=False,wrist_debug_20_#6=False,lat_debug_26_#9=False,groinGlute_debug_27_#7=False)
    DeformRig_Value="(upperarm_up_18_#10=(Rotation=(X=0.0,Y=-0.0,Z=0.0,W=1.0),Translation=(X=0.0,Y=0.0,Z=4),Scale3D=(X=1,Y=1.0,Z=1.0)),upperarm_fwd_fr_19_#11=(Rotation=(X=0.0,Y=-0.0,Z=0.0,W=1.0),Translation=(X=0.0,Y=10. ...(truncated)"
    ExecuteContext=@RigUnit_BeginExecution.ExecutePin
   )
   Deform Belt Stuff = RigVMCollapseNode(ExecutePin=@FN_DeformRig_FNC.ExecuteContext)
   Deform Pant Ends = RigVMCollapseNode(ExecuteContext=@"Deform Belt Stuff".ExecutePin)
   Deform Watches = RigVMCollapseNode(ExecutePin=@"Deform Pant Ends".ExecuteContext)
   Deform Necklaces = RigVMCollapseNode(ExecutePin=@"Deform Watches".ExecutePin)
   Deform Pocket = RigVMCollapseNode(ExecutePin=@"Deform Necklaces".ExecutePin)
   &RigUnit_ParentConstraint(
    AdvancedSettings=(RotationOrderForFilter=XZY)
    Child=(Type=Bone,Name="dyn_zipper")
    Parents="((Item=(Type=Bone,Name=\"clavicle_l\"),Weight=0.0001),(Item=(Type=Bone,Name=\"spine_01\"),Weight=0.0069),(Item=(Type=Bone,Name=\"deform_pec_r\"),Weight=0.0106),(Item=(Type=Bone,Name=\"deform_pec_l\"),Weight=0.0408),(Item=(Type=Bone,Name=\"spine_02\"),Weight=0.1047), ...(truncated)"
    Weight=1
    ExecutePin=@"Deform Pocket".ExecutePin
   )
  ]}
 RigVMModel Rig {nodes [
   n1 = &RigUnit_PrepareForExecution()
   RigUnit_GetTransform = &RigUnit_GetTransform(Item=(Type=Bone,Name="dyn_necklace_03"), bInitial=False)
   n2 = /Script/ControlRig.RigUnit_HierarchyAddBone(
    Name=spawned_necklace_end
    Parent=(Type=Bone,Name="spine_03")
    Space=GlobalSpace
    ExecutePin=@n1.ExecutePin
    Transform=@RigUnit_GetTransform.Transform
   ) {is=struct RigVMUnitNode}
   > RigVMVariableNode(Value=@n2.Item) {variable=Spawned_Necklace_End}
  ]}
]
hierarchy [
 root {loc=<0 0 0>, rot=<0 0 0 1>}
 pelvis {parent=root, loc=<0 -0.7071 93.1191>, rot=<0.0127 0.707 -0.0127 -0.707>}
 spine_01 {parent=pelvis, loc=<6.7494 -0.6388 0>, rot=<0 0 0.034 -0.9994>}
 spine_02 {parent=spine_01, loc=<7.2518 -0.035 0>, rot=<0 0 0.0238 -0.9997>}
 spine_03 {parent=spine_02, loc=<6.0104 0 0>, rot=<0 0 0.0149 -0.9999>}
 spine_04 {parent=spine_03, loc=<8.1024 0 0>, rot=<0 0 -0.0132 -0.9999>}
 spine_05 {parent=spine_04, loc=<10.0146 0 0>, rot=<0 0 -0.0015 -1>}
 neck_01 {parent=spine_05, loc=<11.3138 0.1529 0>, rot=<0 0 -0.2003 -0.9797>}
 neck_02 {parent=neck_01, loc=<4.5627 0 0>, rot=<0 0 0.0504 -0.9987>}
 head {parent=neck_02, loc=<5.3914 0 0>, rot=<0 0 0.115 -0.9934>}
 attach_cape {parent=spine_05, loc=<9.017 -11.714 0>, rot=<0.0616 -0.7044 -0.0616 -0.7044>}
 cape {parent=attach_cape, type=Null, loc=<0 2.0061 0>, rot=<0 0 0 1>}
 attach_backpack {parent=spine_05, loc=<-0.123 -13.473 0>, rot=<-0.0093 -0.707 0.0093 -0.707>}
 Backpack_BR {parent=attach_backpack, type=Null, loc=<0 1.4031 0>, rot=<0 0 0 1>}
 clavicle_l {parent=spine_05, loc=<4.3575 -0.1466 -2.3223>, rot=<0.0465 -0.7289 -0.009 -0.6829>}
 upperarm_l {parent=clavicle_l, loc=<15.5111 0 0>, rot=<0.0226 -0.3234 -0.0088 -0.9459>}
 lowerarm_l {parent=upperarm_l, loc=<26.6514 0 0>, rot=<0.0391 -0.0099 -0.2453 -0.9686>}
 hand_l {parent=lowerarm_l, loc=<22.3491 0 0>, rot=<-0.7081 -0.0507 0.0975 -0.6975>}
 thumb_01_l {parent=hand_l, loc=<1.3118 -0.3191 -2.459>, rot=<0.6546 -0.4042 0.0664 -0.6354>}
 thumb_02_l {parent=thumb_01_l, loc=<5.9336 -0.0448 0.0401>, rot=<-0.0037 0.0213 0.298 -0.9543>}
 thumb_03_l {parent=thumb_02_l, loc=<2.2223 0.3682 -0.0907>, rot=<0.0118 -0.013 0.0066 -0.9998>}
 index_metacarpal_l {parent=hand_l, loc=<1.6535 0.0183 -2.96>, rot=<0.0158 -0.0124 -0.0173 -0.9996>}
 index_01_l {parent=index_metacarpal_l, loc=<8.3008 0.0357 -0.0524>, rot=<-0.0044 0.0485 0.0924 -0.9945>}
 index_02_l {parent=index_01_l, loc=<3.3587 -0.3004 0.0452>, rot=<-0.002 0.0027 0.3225 -0.9466>}
 index_03_l {parent=index_02_l, loc=<2.5578 -0.123 -0.0252>, rot=<-0.002 0.0067 0.2558 -0.9667>}
 middle_metacarpal_l {parent=hand_l, loc=<2.1856 -0.1725 -0.5984>, rot=<0.006 0.0168 -0.0329 -0.9993>}
 middle_01_l {parent=middle_metacarpal_l, loc=<7.3275 -0.2447 0.0129>, rot=<-0.0208 0.0042 0.0858 -0.9961>}
 middle_02_l {parent=middle_01_l, loc=<4.0123 0.1401 0.1175>, rot=<-0.0218 0.0258 0.3512 -0.9357>}
 middle_03_l {parent=middle_02_l, loc=<2.6077 0.0534 0.1373>, rot=<0.0082 0.0188 0.1708 -0.9851>}
 ring_metacarpal_l {parent=hand_l, loc=<2.2074 -0.3625 1.3638>, rot=<-0.0284 0.0722 -0.0458 -0.9959>}
 ring_01_l {parent=ring_metacarpal_l, loc=<7.0295 -0.6197 -0.1446>, rot=<-0.0489 -0.0252 0.1892 -0.9804>}
 ring_02_l {parent=ring_01_l, loc=<3.6174 0.4647 -0.1027>, rot=<0 0 0.2806 -0.9598>}
 ring_03_l {parent=ring_02_l, loc=<2.4152 0.114 0.0199>, rot=<0 0 0.1189 -0.9929>}
 pinky_metacarpal_l {parent=hand_l, loc=<1.7616 -0.9753 3.3814>, rot=<-0.0397 0.0814 -0.069 -0.9935>}
 pinky_01_l {parent=pinky_metacarpal_l, loc=<7.1709 -0.4831 -0.1094>, rot=<-0.0654 -0.0035 0.2242 -0.9723>}
 pinky_02_l {parent=pinky_01_l, loc=<2.8274 0.0893 -0.1043>, rot=<0.0037 0.0011 0.2854 -0.9584>}
 pinky_03_l {parent=pinky_02_l, loc=<2.2046 0.0307 -0.0078>, rot=<-0.0002 0.0098 0.1445 -0.9895>}
 deform_wrist_dn_base_l {parent=hand_l, loc=<0 -2.2342 0>, rot=<0 0 0 -1>}
 deform_wrist_dn_l {parent=deform_wrist_dn_base_l, loc=<0 0 0>, rot=<0 0 0 1>}
 deform_wrist_up_base_l {parent=hand_l, loc=<0 2.2682 0>, rot=<0 0 0 -1>}
 deform_wrist_up_l {parent=deform_wrist_up_base_l, loc=<0 0 0>, rot=<0 0 0 1>}
 lowerarm_twist_02_l {parent=lowerarm_l, loc=<9.6199 0 0>, rot=<0 0 0 -1>}
 lowerarm_twist_01_l {parent=lowerarm_l, loc=<17.8549 0 0>, rot=<0 0 0 -1>}
 deform_watch_l {parent=lowerarm_twist_01_l, loc=<2.3787 0 0.4006>, rot=<0 -0.0283 0 -0.9996>}
 deform_elbow_in_base_l {parent=lowerarm_l, loc=<0.2112 5.259 0.8274>, rot=<0 0 0 -1>}
 deform_elbow_in_l {parent=deform_elbow_in_base_l, loc=<0 0 0>, rot=<0 0 0 1>}
 deform_elbow_out_base_l {parent=lowerarm_l, loc=<0 -4 0>, rot=<0 0 0 -1>}
 deform_elbow_out_l {parent=deform_elbow_out_base_l, loc=<0 0 0>, rot=<0 0 0 1>}
 upperarm_twist_01_l {parent=upperarm_l, loc=<0.0041 0 0>, rot=<0 0 0 -1>}
 deform_upperarm_bk_base_l {parent=upperarm_twist_01_l, loc=<1.0813 -7.2193 -1.9453>, rot=<0 0 0 1>}
 deform_upperarm_bk_l {parent=deform_upperarm_bk_base_l, loc=<0 0 0>, rot=<0 0 0 1>}
 deform_upperarm_fr_base_l {parent=upperarm_twist_01_l, loc=<-0.3691 7.0506 -2.765>, rot=<0 0 0 1>}
 deform_upperarm_fr_l {parent=deform_upperarm_fr_base_l, loc=<0 0 0>, rot=<0 0 0 1>}
 deform_upperarm_up_base_l {parent=upperarm_twist_01_l, loc=<-3.2216 -1.1391 3.3311>, rot=<0 0 0 1>}
 deform_upperarm_up_l {parent=deform_upperarm_up_base_l, loc=<0 0 0>, rot=<0 0 0 1>}
 upperarm_twist_02_l {parent=upperarm_l, loc=<13.7941 0 0>, rot=<0 0 0 -1>}
 clavicle_r {parent=spine_05, loc=<4.3575 -0.1466 2.3223>, rot=<0.7289 0.0465 -0.6829 0.009>}
 upperarm_r {parent=clavicle_r, loc=<-15.5111 0 0>, rot=<0.0226 -0.3234 -0.0088 -0.9459>}
 lowerarm_r {parent=upperarm_r, loc=<-26.6514 0 0>, rot=<0.0391 -0.0099 -0.2453 -0.9686>}
 hand_r {parent=lowerarm_r, loc=<-22.3491 0 0>, rot=<-0.7081 -0.0507 0.0975 -0.6975>}
 thumb_01_r {parent=hand_r, loc=<-1.3118 0.3191 2.459>, rot=<0.6546 -0.4042 0.0664 -0.6354>}
 thumb_02_r {parent=thumb_01_r, loc=<-5.9336 0.0448 -0.0401>, rot=<-0.0037 0.0213 0.298 -0.9543>}
 thumb_03_r {parent=thumb_02_r, loc=<-2.2223 -0.3682 0.0907>, rot=<0.0118 -0.013 0.0066 -0.9998>}
 index_metacarpal_r {parent=hand_r, loc=<-1.6535 -0.0183 2.96>, rot=<0.0158 -0.0124 -0.0173 -0.9996>}
 index_01_r {parent=index_metacarpal_r, loc=<-8.3008 -0.0357 0.0524>, rot=<-0.0044 0.0485 0.0924 -0.9945>}
 index_02_r {parent=index_01_r, loc=<-3.3587 0.3004 -0.0452>, rot=<-0.002 0.0027 0.3225 -0.9466>}
 index_03_r {parent=index_02_r, loc=<-2.5578 0.123 0.0252>, rot=<-0.002 0.0067 0.2558 -0.9667>}
 middle_metacarpal_r {parent=hand_r, loc=<-2.1856 0.1725 0.5984>, rot=<0.006 0.0168 -0.0329 -0.9993>}
 middle_01_r {parent=middle_metacarpal_r, loc=<-7.3275 0.2447 -0.0129>, rot=<-0.0208 0.0042 0.0858 -0.9961>}
 middle_02_r {parent=middle_01_r, loc=<-4.0123 -0.1401 -0.1175>, rot=<-0.0218 0.0258 0.3512 -0.9357>}
 middle_03_r {parent=middle_02_r, loc=<-2.6077 -0.0534 -0.1373>, rot=<0.0082 0.0188 0.1708 -0.9851>}
 ring_metacarpal_r {parent=hand_r, loc=<-2.2074 0.3625 -1.3638>, rot=<-0.0284 0.0722 -0.0458 -0.9959>}
 ring_01_r {parent=ring_metacarpal_r, loc=<-7.0295 0.6197 0.1446>, rot=<-0.0489 -0.0252 0.1892 -0.9804>}
 ring_02_r {parent=ring_01_r, loc=<-3.6174 -0.4647 0.1027>, rot=<0 0 0.2806 -0.9598>}
 ring_03_r {parent=ring_02_r, loc=<-2.4152 -0.114 -0.0199>, rot=<0 0 0.1189 -0.9929>}
 pinky_metacarpal_r {parent=hand_r, loc=<-1.7616 0.9753 -3.3814>, rot=<-0.0397 0.0814 -0.069 -0.9935>}
 pinky_01_r {parent=pinky_metacarpal_r, loc=<-7.1709 0.4831 0.1094>, rot=<-0.0654 -0.0035 0.2242 -0.9723>}
 pinky_02_r {parent=pinky_01_r, loc=<-2.8274 -0.0893 0.1043>, rot=<0.0037 0.0011 0.2854 -0.9584>}
 pinky_03_r {parent=pinky_02_r, loc=<-2.2046 -0.0307 0.0078>, rot=<-0.0002 0.0098 0.1445 -0.9895>}
 deform_wrist_dn_base_r {parent=hand_r, loc=<-0.0004 2.2345 -0.0001>, rot=<0 0 0 -1>}
 deform_wrist_dn_r {parent=deform_wrist_dn_base_r, loc=<0 0 0>, rot=<0 0 0 1>}
 deform_wrist_up_base_r {parent=hand_r, loc=<0.0003 -2.2684 0.0001>, rot=<0 0 0 -1>}
 deform_wrist_up_r {parent=deform_wrist_up_base_r, loc=<0 0 0>, rot=<0 0 0 1>}
 lowerarm_twist_02_r {parent=lowerarm_r, loc=<-9.6199 0 0>, rot=<0 0 0 -1>}
 lowerarm_twist_01_r {parent=lowerarm_r, loc=<-17.8549 0 0>, rot=<0 0 0 -1>}
 deform_watch_r {parent=lowerarm_twist_01_r, loc=<-2.416 0 -0.7219>, rot=<0 -0.0198 0 -0.9998>}
 deform_elbow_in_base_r {parent=lowerarm_r, loc=<-0.2111 -5.2591 -0.8275>, rot=<0 0 0 -1>}
 deform_elbow_in_r {parent=deform_elbow_in_base_r, loc=<-0.0001 0 0.0002>, rot=<0 0 0 1>}
 deform_elbow_out_base_r {parent=lowerarm_r, loc=<0 4 -0.0001>, rot=<0 0 0 -1>}
 deform_elbow_out_r {parent=deform_elbow_out_base_r, loc=<0 0 0>, rot=<0 0 0 1>}
 upperarm_twist_01_r {parent=upperarm_r, loc=<-0.0041 0 0>, rot=<0 0 0 -1>}
 deform_upperarm_bk_base_r {parent=upperarm_twist_01_r, loc=<-1.0814 7.2192 1.9455>, rot=<0 0 0 1>}
 deform_upperarm_bk_r {parent=deform_upperarm_bk_base_r, loc=<0 0 0>, rot=<0 0 0 1>}
 deform_upperarm_fr_base_r {parent=upperarm_twist_01_r, loc=<0.3689 -7.0506 2.7652>, rot=<0 0 0 1>}
 deform_upperarm_fr_r {parent=deform_upperarm_fr_base_r, loc=<0 0 0>, rot=<0 0 0 1>}
 deform_upperarm_up_base_r {parent=upperarm_twist_01_r, loc=<3.2214 1.1391 -3.3308>, rot=<0 0 0 1>}
 deform_upperarm_up_r {parent=deform_upperarm_up_base_r, loc=<0 0 0>, rot=<0 0 0 -1>}
 upperarm_twist_02_r {parent=upperarm_r, loc=<-13.7941 0 0>, rot=<0 0 0 -1>}
 dyn_collar_l {parent=spine_05, loc=<6.9728 7.8136 -5.6404>, rot=<-0.4595 -0.6117 0.0281 -0.6433>}
 dyn_collar_r {parent=spine_05, loc=<6.9733 7.8136 5.6404>, rot=<-0.6117 0.4595 0.6433 0.0281>}
 dyn_necklace_base {parent=spine_05, loc=<2.0495 10.8824 -0.0135>, rot=<-0.1928 -0.9812 -0.0031 0>}
 dyn_necklace_01 {parent=dyn_necklace_base, loc=<0 0 0>, rot=<0 0 0 1>}
 dyn_necklace_02 {parent=dyn_necklace_01, loc=<5.9234 0 0>, rot=<0 0 0.0639 -0.998>}
 dyn_necklace_03 {parent=dyn_necklace_02, loc=<4.6909 0 0>, rot=<0 0 0 1>}
 deform_pec_aim_base_l {parent=spine_04, loc=<7.8732 12.6388 -2.4123>, rot=<0.0102 -0.4602 0.3979 -0.7936>}
 deform_pec_aim_l {parent=deform_pec_aim_base_l, loc=<0 0 0>, rot=<0 0 0 1>}
 deform_pec_base_l {parent=deform_pec_aim_l, loc=<10.766 0 0>, rot=<0 0 0 1>}
 deform_pec_l {parent=deform_pec_base_l, loc=<0.055 0 0>, rot=<0 0 0 1>}
 deform_pec_aim_base_r {parent=spine_04, loc=<7.8727 12.6388 2.4123>, rot=<-0.4602 -0.0102 0.7936 0.3979>}
 deform_pec_aim_r {parent=deform_pec_aim_base_r, loc=<0 0 0>, rot=<0 0 0 1>}
 deform_pec_base_r {parent=deform_pec_aim_r, loc=<-10.7662 -0.0006 -0.0007>, rot=<0 0 0 1>}
 deform_pec_r {parent=deform_pec_base_r, loc=<0 0 0>, rot=<0 0 0 1>}
 deform_pocket_l_base {parent=spine_04, loc=<3.9567 9.923 -11.255>, rot=<0.0332 -0.9649 0.2602 0.0095>}
 dyn_clipboard_base {parent=deform_pocket_l_base, loc=<-2.25 1.1475 0>, rot=<0 0 0 1>}
 dyn_clipboard_01 {parent=dyn_clipboard_base, loc=<0 0 0>, rot=<0 0 0 1>}
 dyn_clipboard_02 {parent=dyn_clipboard_01, loc=<3.1795 0.2638 0>, rot=<0 0 -0.0497 -0.9988>}
 dyn_clipboard_03 {parent=dyn_clipboard_02, loc=<4.2526 0 0>, rot=<0 0 0 1>}
 dyn_zipper {parent=spine_04, loc=<7.4207 13.375 0.1985>, rot=<-0.0921 -0.9958 0 0>}
 thigh_l {parent=pelvis, loc=<-2.5693 -0.0109 -7.7501>, rot=<0.0012 0.0572 0.0155 -0.9982>}
 calf_l {parent=thigh_l, loc=<-39.0484 0 0>, rot=<0.0006 -0.0161 -0.0391 -0.9991>}
 foot_l {parent=calf_l, loc=<-42.561 0.1077 0.4292>, rot=<0.0418 -0.0455 0.0343 -0.9975>}
 ball_l {parent=foot_l, loc=<-6.1599 13.8264 0.6806>, rot=<0.0001 -0.0001 -0.7063 -0.7079>}
 calf_twist_02_l {parent=calf_l, loc=<-18.8064 -0.0967 0.0397>, rot=<0.0237 0.0095 -0.0035 -0.9997>}
 calf_twist_01_l {parent=calf_l, loc=<-35.813 -0.0967 0.0397>, rot=<0.0237 0.0095 -0.0035 -0.9997>}
 deform_pants_lower_l {parent=calf_twist_01_l, loc=<0.9979 0 0>, rot=<0 0 0 -1>}
 deform_knee_in_base_l {parent=calf_l, loc=<1.298 -5.2957 -0.0496>, rot=<0 0 0 -1>}
 deform_knee_in_l {parent=deform_knee_in_base_l, loc=<0 0 0>, rot=<0 0 0 1>}
 deform_knee_out_base_l {parent=calf_l, loc=<0 4 0>, rot=<0 0 0 -1>}
 deform_knee_out_l {parent=deform_knee_out_base_l, loc=<0 0 0>, rot=<0 0 0 1>}
 thigh_twist_01_l {parent=thigh_l, loc=<-21.6303 0 0>, rot=<0.0013 0 0 -1>}
 deform_groin_base_l {parent=thigh_l, loc=<-2.1908 8.1142 -1.7442>, rot=<0 0 0 -1>}
 deform_groin_l {parent=deform_groin_base_l, loc=<0 0 0>, rot=<0 0 0 1>}
 thigh_r {parent=pelvis, loc=<-2.5693 -0.0109 7.7501>, rot=<0.0572 -0.0012 0.9982 0.0155>}
 calf_r {parent=thigh_r, loc=<39.0484 0 0>, rot=<0.0006 -0.0161 -0.0391 -0.9991>}
 foot_r {parent=calf_r, loc=<42.561 -0.1077 -0.4292>, rot=<0.0418 -0.0455 0.0343 -0.9975>}
 ball_r {parent=foot_r, loc=<6.1599 -13.8264 -0.6806>, rot=<0.0001 -0.0001 -0.7063 -0.7079>}
 calf_twist_02_r {parent=calf_r, loc=<18.8064 0.0967 -0.0397>, rot=<0.0237 0.0095 -0.0035 -0.9997>}
 calf_twist_01_r {parent=calf_r, loc=<35.813 0.0967 -0.0397>, rot=<0.0237 0.0095 -0.0035 -0.9997>}
 deform_pants_rower_r {parent=calf_twist_01_r, loc=<-0.9979 0 0>, rot=<0.0138 0 0 -0.9999>}
 deform_knee_in_base_r {parent=calf_r, loc=<-1.298 5.2957 0.0496>, rot=<0 0 0 -1>}
 deform_knee_in_r {parent=deform_knee_in_base_r, loc=<0 0 0>, rot=<0 0 0 1>}
 deform_knee_out_base_r {parent=calf_r, loc=<0 -4 0>, rot=<0 0 0 -1>}
 deform_knee_out_r {parent=deform_knee_out_base_r, loc=<0 0 0>, rot=<0 0 0 1>}
 thigh_twist_01_r {parent=thigh_r, loc=<21.6303 0 0>, rot=<0.0013 0 0 -1>}
 deform_groin_base_r {parent=thigh_r, loc=<2.1908 -8.1142 1.7442>, rot=<0 0 0 -1>}
 deform_groin_r {parent=deform_groin_base_r, loc=<0 0 0>, rot=<0 0 0 1>}
 deform_glute_base_l {parent=pelvis, loc=<-8.8064 -8.8222 -8.4287>, rot=<0.0275 0.0501 -0.4472 -0.8926>}
 deform_glute_l {parent=deform_glute_base_l, loc=<0 0 0>, rot=<0 0 0 1>}
 deform_glute_base_r {parent=pelvis, loc=<-8.8062 -8.8222 8.4287>, rot=<0.0501 -0.0275 0.8926 -0.4472>}
 deform_glute_r {parent=deform_glute_base_r, loc=<0 0 0>, rot=<0 0 0 1>}
 dyn_tool_belt_base {parent=pelvis, loc=<2.4833 1.8722 17.0572>, rot=<-0.011 -0.9914 0.0859 0.0986>}
 dyn_tool_belt {parent=dyn_tool_belt_base, loc=<0 0 0>, rot=<0 0 0 1>}
 dyn_tape_base {parent=dyn_tool_belt, loc=<25.8067 10.1355 0.1393>, rot=<0.0027 -0.9968 -0.0563 -0.0572>}
 dyn_tape {parent=dyn_tape_base, loc=<0 0 0>, rot=<0 0 0 1>}
 dyn_toolbelt_tool_01 {parent=dyn_tool_belt, loc=<9.1358 -10.7553 0.6315>, rot=<0.1693 0.0113 0.0384 -0.9847>}
 dyn_toolbelt_tool_02 {parent=dyn_tool_belt, loc=<9.2519 -4.6188 -0.4721>, rot=<0.0864 0.0309 0.0497 -0.9945>}
 dyn_toolbelt_tool_03 {parent=dyn_tool_belt, loc=<9.4561 2.3763 -1.418>, rot=<-0.0629 0.0514 -0.0197 -0.9965>}
 dyn_toolbelt_tool_04 {parent=dyn_tool_belt, loc=<9.6543 3.8935 -0.5969>, rot=<-0.029 0.0382 0.1068 -0.9931>}
 dyn_belt_clothes_pin {parent=pelvis, loc=<4.4985 10.7745 6.5665>, rot=<0.0236 0.2474 -0.0576 -0.9669>}
 dyn_belt_pouch_base_l {parent=pelvis, loc=<6.1575 7.0719 -14.9229>, rot=<-0.0015 -0.9466 -0.31 -0.0886>}
 dyn_belt_pouch_l {parent=dyn_belt_pouch_base_l, loc=<0 0 0>, rot=<0 0 0 1>}
 dyn_belt_pouch_wire_01_l {parent=dyn_belt_pouch_l, loc=<9.0802 0 -0.5401>, rot=<0 0 0 1>}
 dyn_belt_pouch_wire_02_l {parent=dyn_belt_pouch_wire_01_l, loc=<3.066 0 0>, rot=<0 0 0 1>}
 dyn_belt_base {parent=pelvis, loc=<4.1203 11.7151 1.6283>, rot=<0.0507 0.8057 -0.0996 -0.5817>}
 dyn_belt_01 {parent=dyn_belt_base, loc=<0 0 0>, rot=<0 0 0 1>}
 dyn_belt_02 {parent=dyn_belt_01, loc=<4.4798 0 0>, rot=<0 0 0 1>}
 cr_stomach_target {parent=pelvis, loc=<12 11.298 0>, rot=<0 0 0 1>}
]
variables [
 Skydiving {type=bool, default=False}
 Spawned_Necklace_End {type=FRigElementKey, default=(Type=Bone,Name="")}
]
