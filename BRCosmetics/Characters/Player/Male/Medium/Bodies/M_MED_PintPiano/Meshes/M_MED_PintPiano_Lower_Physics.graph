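# Physics asset for the lower-body mesh of the PintPiano medium male player cosmetic, covering secondary motion of its belt, pouch and tool belt.
# Capsules on pelvis, thighs and right calf give body shapes, alongside small base bodies anchoring each dyn_ chain.
# Tight angular limits hold the two-link belt, pouch, two-link pouch wire, and boxed tool belt with four hanging tools.
# Only the tool belt base is constrained to the pelvis, with free rotation.
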PhysicsAsset /BRCosmetics/Characters/Player/Male/Medium/Bodies/M_MED_PintPiano/Meshes/M_MED_PintPiano_Lower_Physics
bodies {
 calf_r [{shapeType=Capsule, length=34.014, radius=9.4478, center=<23.0111 1.0243 0.0023>, rotation=<rot -0.5012 87.0723 -87.7935>}]
 dyn_belt_01 [{shapeType=Capsule, length=3, radius=2, shapeName=dyn_belt_01_capsule, center=<1.4872 0 0>, rotation=<rot 0 90 90>}]
 dyn_belt_02 [{shapeType=Capsule, length=3, radius=2, shapeName=dyn_belt_02_capsule, center=<2.7583 0 0>, rotation=<rot 0 90 90>}]
 dyn_belt_base [{shapeType=Capsule, length=0.505, radius=0.505, shapeName=dyn_belt_base_capsule, rotation=<rot 0 -90 0>}]
 dyn_belt_pouch_base_l [{shapeType=Capsule, length=0.505, radius=0.505, shapeName=dyn_belt_pouch_base_l_capsule, rotation=<rot 0 -90 0>}]
 dyn_belt_pouch_l [{
   shapeType=Capsule
   length=7.3279
   radius=6.2245
   shapeName=dyn_belt_pouch_l_capsule
   center=<3.7848 0 3.6653>
   rotation=<rot -4.8532 -89.3197 -89.8162>
  }]
 dyn_belt_pouch_wire_01_l [{shapeType=Capsule, length=13.1346, radius=2.8755, shapeName=dyn_belt_pouch_wire_01_l_capsule, center=<2 0 0.7>, rotation=<rot 0 0 90>}]
 dyn_belt_pouch_wire_02_l [{shapeType=Capsule, length=8.2298, radius=3.3106, shapeName=dyn_belt_pouch_wire_02_l_capsule, center=<3.2362 0 0.7>, rotation=<rot 0 0 90>}]
 dyn_tool_belt [{
   shapeType=Box
   extentX=35.3079
   extentY=22.2817
   extentZ=7.5
   shapeName=dyn_tool_belt_box
   center=<10.6705 0 -1.6348>
   rotation=<rot 4.5343 0 0>
  }]
 dyn_tool_belt_base [{shapeType=Capsule, length=0.505, radius=0.505, shapeName=dyn_tool_belt_base_capsule, rotation=<rot 0 -90 0>}]
 dyn_toolbelt_tool_01 [{
   shapeType=Capsule
   length=13.3336
   radius=1.5796
   shapeName=dyn_toolbelt_tool_01_capsule
   center=<-7.6043 0.0094 0.0482>
   rotation=<rot 0 -91.0089 -89.2278>
  }]
 dyn_toolbelt_tool_02 [{
   shapeType=Capsule
   length=11.0784
   radius=4.2163
   shapeName=dyn_toolbelt_tool_02_capsule
   center=<-7.2454 -0.1161 -0.0567>
   rotation=<rot 0 88.5158 -89.9536>
  }]
 dyn_toolbelt_tool_03 [{
   shapeType=Capsule
   length=9.4007
   radius=1.3483
   shapeName=dyn_toolbelt_tool_03_capsule
   center=<-4.8155 0.0062 0.0066>
   rotation=<rot 0.0016 -95.1841 -89.984>
  }]
 dyn_toolbelt_tool_04 [{
   shapeType=Capsule
   length=12.071
   radius=1.5712
   shapeName=dyn_toolbelt_tool_04_capsule
   center=<-6.1672 0.1383 0.0844>
   rotation=<rot -0.1906 -81.4052 -88.8252>
  }]
 pelvis [{shapeType=Capsule, length=3.9794, radius=14, rotation=<rot 0 -90 0>}]
 thigh_l [{shapeType=Capsule, length=41.7723, radius=9.9952, center=<-21.0069 -1.1213 1.5277>, rotation=<rot 0 -87.2269 90>}]
 thigh_r [{shapeType=Capsule, length=31.8873, radius=9.9952, center=<21.0069 1.1213 -1.5277>, rotation=<rot 0 -87.2269 -90>}]
}
constraints [
 {bone1Name=dyn_tool_belt, bone2Name=dyn_tool_belt_base, swing1LimitDegrees=20, swing2LimitDegrees=20, twistLimitDegrees=2}
 {
  bone1Name=dyn_tool_belt_base
  bone2Name=pelvis
  swing1LimitDegrees=45
  swing1Motion=Free
  swing2LimitDegrees=45
  swing2Motion=Free
  twistLimitDegrees=45
  twistMotion=Free
 }
 {bone1Name=dyn_belt_pouch_l, bone2Name=dyn_belt_pouch_base_l, swing1LimitDegrees=6, swing2LimitDegrees=6, twistLimitDegrees=6}
 {bone1Name=dyn_toolbelt_tool_01, bone2Name=dyn_tool_belt, swing1LimitDegrees=12, swing2LimitDegrees=2, twistLimitDegrees=10}
 {bone1Name=dyn_toolbelt_tool_02, bone2Name=dyn_tool_belt, swing1LimitDegrees=12, swing2LimitDegrees=2, twistLimitDegrees=10}
 {bone1Name=dyn_toolbelt_tool_03, bone2Name=dyn_tool_belt, swing1LimitDegrees=12, swing2LimitDegrees=2, twistLimitDegrees=10}
 {bone1Name=dyn_toolbelt_tool_04, bone2Name=dyn_tool_belt, swing1LimitDegrees=8, swing2LimitDegrees=2, twistLimitDegrees=10}
 {bone1Name=dyn_belt_pouch_wire_01_l, bone2Name=dyn_belt_pouch_base_l, swing1LimitDegrees=20, swing2LimitDegrees=20, twistLimitDegrees=20}
 {bone1Name=dyn_belt_pouch_wire_02_l, bone2Name=dyn_belt_pouch_wire_01_l, swing1LimitDegrees=20, swing2LimitDegrees=20, twistLimitDegrees=20}
 {bone1Name=dyn_belt_01, bone2Name=dyn_belt_base, swing1LimitDegrees=4, swing2LimitDegrees=4, twistLimitDegrees=4}
 {bone1Name=dyn_belt_02, bone2Name=dyn_belt_01, swing1LimitDegrees=4, swing2LimitDegrees=4, twistLimitDegrees=4}
]
