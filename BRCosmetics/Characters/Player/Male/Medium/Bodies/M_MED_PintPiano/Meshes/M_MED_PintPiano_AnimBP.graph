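# Animation Blueprint for the PintPiano male body cosmetic, adding secondary motion to clothing and accessories over the parent mesh's copied pose.
# Control Rig feeds pelvis-space RigidBody simulation, then AnimDynamics moves tape, zipper, pants, necklace and clipboard, with bone offsets adjusting belt, clipboard and attachment points.
# A Stage2 variant tag sets Mask Active, switching collars between fixed rotations and dynamics.
# Events also update gravity override and a backbling-with-mask flag.
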AnimBlueprint /BRCosmetics/Characters/Player/Male/Medium/Bodies/M_MED_PintPiano/Meshes/M_MED_PintPiano_AnimBP
target_skeleton=&/M_MED_PintPiano_Skeleton
graphs [
 AnimGraph {nodes [
   n1 = "Transform (Modify) Bone - Bone: dyn_collar_l"!(
    ComponentPose=Local To Component(LocalPose=Use cached pose 'PostRBAN'())
    bAlphaBoolEnabled:bool=True
    Translation="0,0,0"
    Rotation="-4.881,-22.874,9.450"
    Scale="1,1,1"
   ) -> (Pose) {
    alphaInputType=Bool
    rotationMode=BMM_Additive
    rotationSpace=BCS_BoneSpace
    alphaBoolBlend {blendOption=Linear}
    boneToModify {boneName=dyn_collar_l}
    rotation=<rot -4.881 -22.874 9.45>
   }
   n2 = "Transform (Modify) Bone - Bone: dyn_collar_r"!(ComponentPose=@n1, bAlphaBoolEnabled:bool=True, Translation="0,0,0", Rotation="0.483,-46.255,-5.218", Scale="1,1,1") -> (Pose) {
    alphaInputType=Bool
    rotationMode=BMM_Additive
    rotationSpace=BCS_BoneSpace
    alphaBoolBlend {blendOption=Linear}
    boneToModify {boneName=dyn_collar_r}
    rotation=<rot 0.483 -46.255 -5.2181>
   }
   n3 = Component To Local(ComponentPose=@n2)
   n4 = Anim Dynamics - Bone: dyn_collar_l(ComponentPose=Local To Component(LocalPose=Use cached pose 'PostRBAN'()), Alpha=1, ExternalForce="0,0,0") {
    angularDampingOverride=1
    angularSpringConstant=20000
    bAngularSpring=true
    bOverrideAngularDamping=true
    bUseGravityOverride=true
    gravityScale=0.35
    simulationSpace=RootRelative
    boundBone {boneName=dyn_collar_l}
    componentAppliedLinearAccClamp=<0 0 0>
    gravityOverride {z=-980}
    physicsBodyDefinitions [{
      boundBone {boneName=dyn_collar_l}
      boxExtents=<20 30 30>
      constraintSetup {angularLimitsMax=<15 15 15>, angularLimitsMin=<-15 -15 -15>, angularTarget=<1 0 0>}
      localJointOffset=<4 0 0>
     }]
    simSpaceSettings {maxAngularAcceleration=300, maxAngularVelocity=770, simSpaceAngularAlpha=1}
   }
   n5 = Anim Dynamics - Bone: dyn_collar_r(ComponentPose=@n4, Alpha=1, ExternalForce="0,0,0") {
    angularDampingOverride=1
    angularSpringConstant=20000
    bAngularSpring=true
    bOverrideAngularDamping=true
    bUseGravityOverride=true
    gravityScale=0.35
    simulationSpace=RootRelative
    boundBone {boneName=dyn_collar_r}
    componentAppliedLinearAccClamp=<0 0 0>
    gravityOverride {z=-980}
    physicsBodyDefinitions [{
      boundBone {boneName=dyn_collar_r}
      boxExtents=<20 30 30>
      constraintSetup {angularLimitsMax=<15 15 15>, angularLimitsMin=<-15 -15 -15>, angularTarget=<1 0 0>}
      localJointOffset=<4 0 0>
     }]
    simSpaceSettings {maxAngularAcceleration=300, maxAngularVelocity=770, simSpaceAngularAlpha=1}
   }
   n6 = Component To Local(ComponentPose=@n5)
   n7 = Get Mask Active() -> (Mask Active:bool)
   n8 = Anim Dynamics - Bone: dyn_tape(
    ComponentPose=Local To Component(LocalPose=Blend Poses by bool(BlendPose_0=@n3, BlendPose_1=@n6, BlendTime_0=0, BlendTime_1=0, bActiveValue=@n7))
    Alpha=1
    ExternalForce="0,0,0"
   ) {
    angularDampingOverride=1
    angularSpringConstant=20000
    bLinearSpring=true
    bOverrideLinearDamping=true
    bUseGravityOverride=true
    gravityScale=0.35
    linearDampingOverride=0.2
    linearSpringConstant=50
    simulationSpace=RootRelative
    boundBone {boneName=dyn_tape}
    componentAppliedLinearAccClamp=<0 0 0>
    gravityOverride {z=-200}
    physicsBodyDefinitions [{
      boundBone {boneName=dyn_tape}
      boxExtents=<30 20 20>
      constraintSetup {angularLimitsMax=<0 0 0>, angularLimitsMin=<0 0 0>, angularTarget=<0 0 0>, linearAxesMax=<3 0 0>}
      localJointOffset=<4 0 0>
     }]
    simSpaceSettings {maxAngularAcceleration=300, maxAngularVelocity=770, simSpaceAngularAlpha=1}
   }
   n9 = Anim Dynamics - Bone: dyn_tape(ComponentPose=@n8, Alpha=1, ExternalForce="0,0,0") {
    angularDampingOverride=0.95
    angularSpringConstant=10000
    bAngularSpring=true
    bOverrideAngularDamping=true
    bUseGravityOverride=true
    gravityScale=0.35
    simulationSpace=RootRelative
    boundBone {boneName=dyn_tape}
    componentAppliedLinearAccClamp=<0 0 0>
    gravityOverride {z=-980}
    physicsBodyDefinitions [{
      boundBone {boneName=dyn_tape}
      boxExtents=<20 30 30>
      constraintSetup {angularLimitsMax=<30 30 30>, angularLimitsMin=<-30 -30 -30>, angularTarget=<1 0 0>}
      localJointOffset=<4 0 0>
     }]
    simSpaceSettings {maxAngularAcceleration=300, maxAngularVelocity=770, simSpaceAngularAlpha=1}
   }
   n10 = Anim Dynamics - Bone: dyn_zipper(ComponentPose=@n9, Alpha=1) {
    angularDampingOverride=1
    angularSpringConstant=20000
    bAngularSpring=true
    bOverrideAngularDamping=true
    bUseGravityOverride=true
    gravityScale=0.35
    simulationSpace=RootRelative
    boundBone {boneName=dyn_zipper}
    componentAppliedLinearAccClamp=<0 0 0>
    gravityOverride {z=-980}
    physicsBodyDefinitions [{
      boundBone {boneName=dyn_zipper}
      boxExtents=<20 30 30>
      constraintSetup {angularLimitsMax=<30 30 15>, angularLimitsMin=<-30 -30 0>, angularTarget=<1 0 0>}
      localJointOffset=<4 0 0>
     }]
    simSpaceSettings {maxAngularAcceleration=300, maxAngularVelocity=770, simSpaceAngularAlpha=1}
   }
   n11 = Anim Dynamics - Bone: deform_pants_lower_l!(ComponentPose=@n10, Alpha=1, GravityOverride:Vector="0,0,-980", ExternalForce="0,0,0") -> (Pose) {
    angularDampingOverride=1
    angularSpringConstant=20000
    bAngularSpring=true
    bOverrideAngularDamping=true
    bUseGravityOverride=true
    gravityScale=0.35
    simulationSpace=RootRelative
    boundBone {boneName=deform_pants_lower_l}
    componentAppliedLinearAccClamp=<0 0 0>
    gravityOverride {z=-980}
    physicsBodyDefinitions [{
      boundBone {boneName=deform_pants_lower_l}
      boxExtents=<20 30 30>
      constraintSetup {angularLimitsMax=<5 5 5>, angularLimitsMin=<-5 -5 -5>, angularTarget=<1 0 0>}
      localJointOffset=<-3 0 0>
     }]
    simSpaceSettings {maxAngularAcceleration=300, maxAngularVelocity=770, simSpaceAngularAlpha=1}
   }
   n12 = Anim Dynamics - Bone: deform_pants_rower_r!(ComponentPose=@n11, Alpha=1, GravityOverride:Vector="0,0,-980", ExternalForce="0,0,0") -> (Pose) {
    angularDampingOverride=1
    angularSpringConstant=20000
    bAngularSpring=true
    bOverrideAngularDamping=true
    bUseGravityOverride=true
    gravityScale=0.35
    simulationSpace=RootRelative
    boundBone {boneName=deform_pants_rower_r}
    componentAppliedLinearAccClamp=<0 0 0>
    gravityOverride {z=-980}
    physicsBodyDefinitions [{
      boundBone {boneName=deform_pants_rower_r}
      boxExtents=<20 30 30>
      constraintSetup {angularLimitsMax=<5 5 5>, angularLimitsMin=<-5 -5 -5>, angularTarget=<1 0 0>}
      localJointOffset=<3 0 0>
     }]
    simSpaceSettings {maxAngularAcceleration=300, maxAngularVelocity=770, simSpaceAngularAlpha=1}
   }
   n13 = Anim Dynamics - Bone: dyn_necklace_01(ComponentPose=@n12, Alpha=1, ExternalForce="0,0,0") {
    angularDampingOverride=0.999
    angularSpringConstant=8000
    bAngularSpring=true
    bOverrideAngularDamping=true
    bUseGravityOverride=true
    gravityScale=0.35
    simulationSpace=RootRelative
    boundBone {boneName=dyn_necklace_01}
    gravityOverride {z=-980}
    physicsBodyDefinitions [{
      boundBone {boneName=dyn_necklace_01}
      boxExtents=<30 30 30>
      constraintSetup {angularLimitsMax=<1 3 20>, angularLimitsMin=<-1 -3 0>, angularTarget=<1 0 0>}
      localJointOffset=<4 0 0>
     }]
    relativeSpaceBone {boneName=dyn_necklace_base}
    simSpaceSettings {maxAngularAcceleration=300, maxAngularVelocity=1200, simSpaceAngularAlpha=1}
   }
   n14 = Anim Dynamics - Bone: dyn_necklace_02!(ComponentPose=@n13, Alpha=1, GravityOverride:Vector="0,0,-980", ExternalForce="0,0,0") -> (Pose) {
    angularDampingOverride=0.999
    angularSpringConstant=8000
    bAngularSpring=true
    bOverrideAngularDamping=true
    bUseGravityOverride=true
    gravityScale=0.35
    simulationSpace=RootRelative
    boundBone {boneName=dyn_necklace_02}
    gravityOverride {z=-980}
    physicsBodyDefinitions [{
      boundBone {boneName=dyn_necklace_02}
      boxExtents=<30 30 30>
      constraintSetup {angularLimitsMax=<1 4 24>, angularLimitsMin=<-1 -4 0>, angularTarget=<1 0 0>}
      localJointOffset=<4 0 0>
     }]
    relativeSpaceBone {boneName=dyn_necklace_01}
    simSpaceSettings {maxAngularAcceleration=300, maxAngularVelocity=1200, simSpaceAngularAlpha=1}
   }
   n15 = Anim Dynamics - Bone: dyn_necklace_03!(ComponentPose=@n14, Alpha=1, GravityOverride:Vector="0,0,-980", ExternalForce="0,0,0") -> (Pose) {
    angularDampingOverride=0.999
    angularSpringConstant=8000
    bAngularSpring=true
    bOverrideAngularDamping=true
    bUseGravityOverride=true
    gravityScale=0.35
    simulationSpace=RootRelative
    boundBone {boneName=dyn_necklace_03}
    gravityOverride {z=-980}
    physicsBodyDefinitions [{
      boundBone {boneName=dyn_necklace_03}
      boxExtents=<30 30 30>
      constraintSetup {angularLimitsMax=<2 24 32>, angularLimitsMin=<-2 -24 0>, angularTarget=<1 0 0>}
      localJointOffset=<4 0 0>
     }]
    relativeSpaceBone {boneName=dyn_necklace_02}
    simSpaceSettings {maxAngularAcceleration=300, maxAngularVelocity=1200, simSpaceAngularAlpha=1}
   }
   n16 = Anim Dynamics - Bone: dyn_clipboard_02!(ComponentPose=@n15, Alpha=1, GravityOverride:Vector="0,0,-980", ExternalForce="0,0,0") -> (Pose) {
    angularDampingOverride=0.99
    angularSpringConstant=2000
    bAngularSpring=true
    bOverrideAngularDamping=true
    bUseGravityOverride=true
    gravityScale=0.35
    simulationSpace=RootRelative
    boundBone {boneName=dyn_clipboard_02}
    componentAppliedLinearAccClamp=<0 0 700>
    componentLinearAccScale {z=1}
    componentLinearVelScale {z=1}
    gravityOverride {z=-980}
    physicsBodyDefinitions [{
      boundBone {boneName=dyn_clipboard_02}
      boxExtents=<20 20 20>
      constraintSetup {angularLimitsMax=<2 6 15>, angularLimitsMin=<-2 -6 0>, angularTarget=<1 0 0>}
      localJointOffset=<2 0 0>
     }]
    relativeSpaceBone {boneName=spine_04}
    simSpaceSettings {maxAngularAcceleration=100, maxAngularVelocity=770, simSpaceAngularAlpha=1}
   }
   n17 = Anim Dynamics - Bone: dyn_clipboard_03!(ComponentPose=@n16, Alpha=1, GravityOverride:Vector="0,0,-980", ExternalForce="0,0,0") -> (Pose) {
    angularDampingOverride=0.99
    angularSpringConstant=2000
    bAngularSpring=true
    bOverrideAngularDamping=true
    bUseGravityOverride=true
    gravityScale=0.35
    simulationSpace=RootRelative
    boundBone {boneName=dyn_clipboard_03}
    componentAppliedLinearAccClamp=<0 0 700>
    componentLinearAccScale {z=1}
    componentLinearVelScale {z=1}
    gravityOverride {z=-980}
    physicsBodyDefinitions [{
      boundBone {boneName=dyn_clipboard_03}
      boxExtents=<20 20 20>
      constraintSetup {angularLimitsMax=<2 10 15>, angularLimitsMin=<-2 -10 0>, angularTarget=<1 0 0>}
      localJointOffset=<2 0 0>
     }]
    relativeSpaceBone {boneName=spine_04}
    simSpaceSettings {maxAngularAcceleration=100, maxAngularVelocity=770, simSpaceAngularAlpha=1}
   }
   n18 = "Transform (Modify) Bone - Bone: dyn_clipboard_02"!(ComponentPose=@n17, bAlphaBoolEnabled:bool=True, Translation="0,0,0", Rotation="0,30,0", Scale="1,1,1") -> (Pose) {
    alphaInputType=Bool
    rotationMode=BMM_Additive
    rotationSpace=BCS_BoneSpace
    alphaBoolBlend {blendInTime=0.3, blendOutTime=0.2}
    boneToModify {boneName=dyn_clipboard_02}
    rotation {yaw=30}
   }
   n19 = "Transform (Modify) Bone - Bone: dyn_clipboard_03"!(ComponentPose=@n18, bAlphaBoolEnabled:bool=True, Translation="0,0,0", Rotation="5,40,0", Scale="1,1,1") -> (Pose) {
    alphaInputType=Bool
    rotationMode=BMM_Additive
    rotationSpace=BCS_BoneSpace
    alphaBoolBlend {blendInTime=0.5, blendOutTime=0.2}
    boneToModify {boneName=dyn_clipboard_03}
    rotation {pitch=5, yaw=40}
   }
   n20 = "Transform (Modify) Bone - Bone: attach_backpack"!(ComponentPose=@n19, bAlphaBoolEnabled:bool=True, Translation="0,-1,0", Rotation="0,0,0", Scale="1,1,1") -> (Pose) {
    alphaInputType=Bool
    translationMode=BMM_Additive
    translationSpace=BCS_BoneSpace
    boneToModify {boneName=attach_backpack}
    translation {y=-1}
   }
   n21 = "Transform (Modify) Bone - Bone: attach_cape"!(ComponentPose=@n20, bAlphaBoolEnabled:bool=True, Translation="0.154,-5.204,-3.376", Rotation="0,0,0", Scale="1,1,1") -> (Pose) {
    alphaInputType=Bool
    translationMode=BMM_Additive
    translationSpace=BCS_BoneSpace
    boneToModify {boneName=attach_cape}
    translation=<0.154 -5.204 -3.376>
   }
   Output Pose(Result=Component To Local(ComponentPose=@n21)) {layerGroup=None}
   n22 = Control Rig!(Source=Copy Pose From Mesh() {bUseAttachedParent=true}, Alpha, Skydiving:bool=Get bIsSkyDiving()) -> (Pose) {controlRigAssetReference {blueprintRigClass=&/M_MED_PintPiano_CtrlRig.M_MED_PintPiano_CtrlRig_C}}
   Save cached pose 'Post Deform'(Pose=@n22) {cacheName=Post Deform}
   n23 = RigidBody!(ComponentPose=Local To Component(LocalPose=Use cached pose 'Post Deform'()), Alpha, OverrideWorldGravity:Vector="0,0,-980", ExternalForce) -> (Pose) {
    bOverrideWorldGravity=true
    overridePhysicsAsset=&/M_MED_PintPiano_Lower_Physics
    simulationSpace=BaseBoneSpace
    baseBoneRef {boneName=pelvis}
    overrideWorldGravity {z=-980}
    simSpaceSettings {dampingAlpha=0, maxAngularAcceleration=8, maxAngularVelocity=100, maxLinearAcceleration=8, maxLinearVelocity=1000, worldAlpha=1}
   }
   n24 = "Transform (Modify) Bone - Bone: dyn_belt_01"(ComponentPose=@n23, Alpha=1, Translation="0,0,0", Rotation="-8,0,0", Scale="1,1,1") {rotationMode=BMM_Additive, rotationSpace=BCS_BoneSpace, boneToModify {boneName=dyn_belt_01}, rotation {pitch=-8}}
   Save cached pose 'PostRBAN'(Pose=Component To Local(ComponentPose="Transform (Modify) Bone - Bone: dyn_belt_02"(ComponentPose=@n24, Alpha=1, Translation="0,0,0", Rotation="-8,0,0", Scale="1,1,1") {rotationMode=BMM_Additive, rotationSpace=BCS_BoneSpace, boneToModify {boneName=dyn_belt_02}, rotation {pitch=-8}})) {cacheName=PostRBAN}
  ]}
 EventGraph {nodes [
   Event BlueprintUpdateAnimation()
   > n1 = "Gravity Override State Function (Parameter Asset) (Nativized)"(
    CharacterPartAnimInstance=Self-Reference()
    GameTimeInSeconds=GetGameTimeInSeconds()
    Parameters=&/M_MED_PintPiano_Parameters
   )
   > Set Gravity Override(Gravity Override=@n1)
   Event OnCharacterCustomizationLoaded()
   > n2 = Sequence()
   > n3 = Branch(Condition=CharacterItemDefinitionHasAnyMetaTag(Tags=Make Array("[0]"=(TagName="Cosmetics.Variant.Property.Stage2"))))
   > Set Mask Active(Mask Active:bool=true) -> (Output_Get:bool)
   Set Mask Active(execute=@n3.else, Mask Active=false)
   n4 = CheckForBackbling()
   Set Backbling and Mask Equipped(
    execute=@n2.then_1
    Backbling and Mask Equipped:bool=AND Boolean(A=OR Boolean!(A=@n4, B=@n4.bIsBackblingACape, C:bool=@n4.bIsBackblingLong) -> (ReturnValue), B=Get Mask Active())
   ) -> (Output_Get:bool)
  ]}
]
variables [
 Backbling and Mask Equipped {type=boolean}
 Gravity Override {type=Vector, default=<0 0 -980>}
 Mask Active {type=boolean}
 Skydiving {type=boolean}
]
