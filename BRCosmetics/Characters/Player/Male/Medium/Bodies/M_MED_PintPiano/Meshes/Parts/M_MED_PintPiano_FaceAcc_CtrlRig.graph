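# Control Rig for the PintPiano medium male face accessory that adjusts dynamic hair base bones according to head pose.
# Two spherical pose readers track the head bone relative to spine_05, each with its own rotation offset and active and falloff regions.
# Each reader's output weights a following Modify Transforms node that offsets the hair base bones, so the two corrective poses run in sequence.
# Both readers have debug drawing switched on.
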ControlRigBlueprint /BRCosmetics/Characters/Player/Male/Medium/Bodies/M_MED_PintPiano/Meshes/Parts/M_MED_PintPiano_FaceAcc_CtrlRig
graphs [
 RigVMFunctionLibrary {links [], nodes []}
 RigVMModel {nodes [
   RigUnit_BeginExecution = &RigUnit_BeginExecution()
   n1 = &RigUnit_SphericalPoseReader(
    ActiveRegionScaleFactors=(PositiveWidth=1,NegativeWidth=5,PositiveHeight=5,NegativeHeight=5)
    ActiveRegionSize=0.1
    Debug=(bDrawDebug=True,bDraw2D=False,bDrawLocalAxes=False,DebugScale=25,DebugSegments=20,DebugThickness=0.25)
    DriverAxis=(X=1,Y=0,Z=0)
    DriverItem=(Type=Bone,Name="head")
    FalloffRegionScaleFactors=(PositiveWidth=1,NegativeWidth=1,PositiveHeight=1,NegativeHeight=1)
    FalloffSize=0.5
    FlipHeightScaling=False
    FlipWidthScaling=False
    OptionalParentItem=(Type=Bone,Name="spine_05")
    RotationOffset=(X=-90,Y=0,Z=0)
    ExecutePin=@RigUnit_BeginExecution.ExecutePin
   )
   > &RigUnit_ModifyTransforms(
    ItemToModify="((Item=(Type=Bone,Name=\"dyn_hair_01_base\"),Transform=(Rotation=(X=0,Y=0,Z=0.2948,W=0.9556),Translation=(X=-5.5362,Y=-1.3313,Z=0),Scale3D=(X=0.9724,Y=1,Z=1))),(Item=(Type=Bone,Name=\"dyn_hair_01_l_base\"),Transform=(Rotation=(X=0,Y=-0.1458,Z=0,W=0.9893),Trans ...(truncated)"
    WeightMinimum=0
    Weight=@n1.OutputParam
   )
   > n2 = &RigUnit_SphericalPoseReader(
    ActiveRegionScaleFactors=(PositiveWidth=1,NegativeWidth=3,PositiveHeight=1.5,NegativeHeight=1.5)
    ActiveRegionSize=0.1
    Debug=(bDrawDebug=True,bDraw2D=False,bDrawLocalAxes=False,DebugScale=25,DebugSegments=20,DebugThickness=0.25)
    DriverAxis=(X=1,Y=0,Z=0)
    DriverItem=(Type=Bone,Name="head")
    FalloffRegionScaleFactors=(PositiveWidth=1,NegativeWidth=0.5,PositiveHeight=0.5,NegativeHeight=0.5)
    FalloffSize=0.3
    FlipHeightScaling=False
    FlipWidthScaling=False
    OptionalParentItem=(Type=Bone,Name="spine_05")
    RotationOffset=(X=-60,Y=10,Z=0)
   )
   > &RigUnit_ModifyTransforms(
    ItemToModify="((Item=(Type=Bone,Name=\"dyn_hair_01_base\"),Transform=(Rotation=(X=0,Y=0,Z=0.0205,W=0.9998),Translation=(X=-0.6922,Y=-0.0276,Z=0),Scale3D=(X=1,Y=1,Z=1))),(Item=(Type=Bone,Name=\"dyn_hair_01_l_base\"),Transform=(Rotation=(X=0,Y=0,Z=0,W=1),Translation=(X=0,Y=0, ...(truncated)"
    WeightMinimum=0
    Weight=@n2.OutputParam
   )
  ]}
]
hierarchy [
 root {loc=<0 0 0>, rot=<0 0 0 1>}
 pelvis {parent=root, loc=<0 -0.7071 93.1191>, rot=<0.0127 0.707 -0.0127 -0.707>}
 spine_01 {parent=pelvis, loc=<6.7494 -0.6388 0>, rot=<0 0 0.034 -0.9994>}
 spine_02 {parent=spine_01, loc=<7.2518 -0.035 0>, rot=<0 0 0.0238 -0.9997>}
 spine_03 {parent=spine_02, loc=<6.0104 0 0>, rot=<0 0 0.0149 -0.9999>}
 spine_04 {parent=spine_03, loc=<8.1024 0 0>, rot=<0 0 -0.0132 -0.9999>}
 spine_05 {parent=spine_04, loc=<10.0146 0 0>, rot=<0 0 -0.0015 -1>}
 neck_01 {parent=spine_05, loc=<11.3138 0.1529 0>, rot=<0 0 -0.2003 -0.9797>}
 neck_02 {parent=neck_01, loc=<4.5627 0 0>, rot=<0 0 0.0504 -0.9987>}
 head {parent=neck_02, loc=<5.3914 0 0>, rot=<0 0 0.115 -0.9934>}
 dyn_hair_01_l_base {parent=head, loc=<12.4691 -0.328 -6.5093>, rot=<0.015 -0.5549 0.7575 -0.3436>}
 dyn_hair_01_l {parent=dyn_hair_01_l_base, loc=<0 0 0>, rot=<0 0 0 1>}
 dyn_hair_02_l {parent=dyn_hair_01_l, loc=<6.8644 0 0>, rot=<-0.0277 -0.0899 0.2937 -0.9513>}
 dyn_hair_03_l {parent=dyn_hair_02_l, loc=<4.6909 0 0>, rot=<0 0 0.0069 -1>}
 dyn_hair_01_base {parent=head, loc=<13.3094 -3.2582 0>, rot=<0.0044 0.0016 0.9408 -0.3389>}
 dyn_hair_01_rear {parent=dyn_hair_01_base, loc=<0 0 0>, rot=<0 0 0 1>}
 dyn_hair_02_rear {parent=dyn_hair_01_rear, loc=<6.8644 0 0>, rot=<0 0 0.2174 -0.9761>}
 dyn_hair_03_rear {parent=dyn_hair_02_rear, loc=<4.6909 0 0>, rot=<0 0 0.0069 -1>}
 dyn_hair_01_r_base {parent=head, loc=<12.4689 -0.328 6.5093>, rot=<0.5549 0.015 -0.3436 -0.7575>}
 dyn_hair_01_r {parent=dyn_hair_01_r_base, loc=<0 0 0>, rot=<0 0 0 1>}
 dyn_hair_02_r {parent=dyn_hair_01_r, loc=<-6.8641 -0.0002 -0.0002>, rot=<-0.0277 -0.0899 0.2937 -0.9513>}
 dyn_hair_03_r {parent=dyn_hair_02_r, loc=<-4.691 0 0>, rot=<0 0 0.0069 -1>}
 clavicle_l {parent=spine_05, loc=<4.3575 -0.1466 -2.3223>, rot=<0.0465 -0.7289 -0.009 -0.6829>}
 clavicle_r {parent=spine_05, loc=<4.3575 -0.1466 2.3223>, rot=<0.7289 0.0465 -0.6829 0.009>}
]
