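# Physics asset for the M_MED_PintPiano face accessory mesh, defining capsule bodies and joints for its dynamic hair bones.
# Left, right and rear chains each run three bones from a base bone, with tight swing and twist limits.
# Both side tips also join the rear tip through constraints with free swing and twist, cross-linking the chains.
# Head, clavicle and upper-spine capsules have no constraints, apparently serving as colliders for the hair.
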PhysicsAsset /BRCosmetics/Characters/Player/Male/Medium/Bodies/M_MED_PintPiano/Meshes/Parts/M_MED_PintPiano_FaceAcc_Physics
bodies {
 clavicle_l [{shapeType=Capsule, length=7.4876, radius=8.9592, shapeName=clavicle_l_capsule, center=<9.0152 0 0>, rotation=<rot 0 -90 106.3567>}]
 clavicle_r [{shapeType=Capsule, length=7.4876, radius=8.9592, shapeName=clavicle_r_capsule, center=<-9.0152 0 0>, rotation=<rot 0 -90 -73.6433>}]
 dyn_hair_01_base [{shapeType=Capsule, length=0.505, radius=0.505, shapeName=dyn_hair_01_base_capsule, rotation=<rot 0 -90 0>}]
 dyn_hair_01_l [{shapeType=Capsule, length=8.0957, radius=3.9043, shapeName=dyn_hair_01_l_capsule, center=<2 2 0>}]
 dyn_hair_01_l_base [{shapeType=Capsule, length=0.505, radius=0.505, shapeName=dyn_hair_01_l_base_capsule, rotation=<rot 0 -90 0>}]
 dyn_hair_01_r [{shapeType=Capsule, length=8.2131, radius=3.7869, shapeName=dyn_hair_01_r_capsule, center=<-2 -2 0>, rotation=<rot 0 0 -180>}]
 dyn_hair_01_r_base [{shapeType=Capsule, length=0.505, radius=0.505, shapeName=dyn_hair_01_r_base_capsule, rotation=<rot 0 -90 0>}]
 dyn_hair_01_rear [{shapeType=Capsule, length=9.8204, radius=2.1796, shapeName=dyn_hair_01_rear_capsule, center=<2 2 0>}]
 dyn_hair_02_l [{shapeType=Capsule, length=4.5288, radius=4, shapeName=dyn_hair_02_l_capsule, center=<2 2 0>}]
 dyn_hair_02_r [{shapeType=Capsule, length=4.2478, radius=4, shapeName=dyn_hair_02_r_capsule, center=<-2 -2 0>, rotation=<rot 0 0 -180>}]
 dyn_hair_02_rear [{shapeType=Capsule, length=8.4534, radius=3.5465, shapeName=dyn_hair_02_rear_capsule, center=<2 2 0>}]
 dyn_hair_03_l [{shapeType=Capsule, length=3.3869, radius=4, shapeName=dyn_hair_03_l_capsule, center=<2 2 0>}]
 dyn_hair_03_r [{shapeType=Capsule, length=1.8409, radius=4, shapeName=dyn_hair_03_r_capsule, center=<-2 -2 0>, rotation=<rot 0 0 -180>}]
 dyn_hair_03_rear [{shapeType=Capsule, length=8, radius=4, shapeName=dyn_hair_03_rear_capsule, center=<5 1 0>}]
 head [
  {shapeType=Capsule, length=10.5082, radius=6.494, shapeName=head_capsule, center=<1.5619 0 0>, rotation=<rot 90 0 0>}
  {
   shapeType=Capsule
   length=5.2477
   radius=2.7954
   shapeName=head_capsule
   center=<2.3149 1.6552 -6.5253>
   rotation=<rot -86.6608 179.9985 -162.2651>
  }
  {
   shapeType=Capsule
   length=4.2286
   radius=2.7553
   shapeName=head_capsule
   center=<2.3951 2.1668 7.7507>
   rotation=<rot 78.8629 -77.2928 -93.6513>
  }
 ]
 spine_05 [{shapeType=Capsule, length=12.9867, radius=14.3182, shapeName=spine_05_capsule, center=<2.322 0 0>, rotation=<rot 0 -90 0>}]
}
constraints [
 {bone1Name=dyn_hair_01_rear, bone2Name=dyn_hair_01_base, swing1LimitDegrees=4, swing2LimitDegrees=4, twistLimitDegrees=4}
 {bone1Name=dyn_hair_02_rear, bone2Name=dyn_hair_01_rear, swing1LimitDegrees=4, swing2LimitDegrees=4, twistLimitDegrees=4}
 {bone1Name=dyn_hair_03_rear, bone2Name=dyn_hair_02_rear, swing1LimitDegrees=6, swing2LimitDegrees=6, twistLimitDegrees=6}
 {bone1Name=dyn_hair_01_l, bone2Name=dyn_hair_01_l_base, swing1LimitDegrees=8, swing2LimitDegrees=8, twistLimitDegrees=4}
 {bone1Name=dyn_hair_02_l, bone2Name=dyn_hair_01_l, swing1LimitDegrees=8, swing2LimitDegrees=8, twistLimitDegrees=4}
 {bone1Name=dyn_hair_03_l, bone2Name=dyn_hair_02_l, swing1LimitDegrees=8, swing2LimitDegrees=8, twistLimitDegrees=6}
 {bone1Name=dyn_hair_01_r, bone2Name=dyn_hair_01_r_base, swing1LimitDegrees=8, swing2LimitDegrees=8, twistLimitDegrees=4}
 {bone1Name=dyn_hair_02_r, bone2Name=dyn_hair_01_r, swing1LimitDegrees=8, swing2LimitDegrees=8, twistLimitDegrees=4}
 {bone1Name=dyn_hair_03_r, bone2Name=dyn_hair_02_r, swing1LimitDegrees=8, swing2LimitDegrees=8, twistLimitDegrees=6}
 {
  bone1Name=dyn_hair_03_r
  bone2Name=dyn_hair_03_rear
  swing1LimitDegrees=45
  swing1Motion=Free
  swing2LimitDegrees=45
  swing2Motion=Free
  twistLimitDegrees=45
  twistMotion=Free
 }
 {
  bone1Name=dyn_hair_03_l
  bone2Name=dyn_hair_03_rear
  swing1LimitDegrees=45
  swing1Motion=Free
  swing2LimitDegrees=45
  swing2Motion=Free
  twistLimitDegrees=45
  twistMotion=Free
 }
]
